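; BooleanOp is a sample PCG graph that uses Geometry Script nodes to build a slab with four openings in a 2x2 grid.
; The engine's basic cube becomes a dynamic mesh that feeds five transforms: one shapes the slab, four place smaller cutters deep enough to pierce it.
; A Boolean Operation subtracts each cutter from the slab sequentially, and Spawn Dynamic Mesh outputs the result.
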(graph BooleanOp
  (bind StaticMeshToDynamicMeshElement_0 (Mesh%20To%20Dynamic%20Mesh%20Element :Mesh /Engine/BasicShapes/Cube.Cube))
  (bind DynamicMeshTransform_1 (Dynamic%20Mesh%20Transform :Transform (struct :Rotation (0.0 -0.0 0.7071067811865476 0.7071067811865477) :Translation (0.0 0.0 1e+02) :Scale3D (0.5 3.0 3.0))))
  (bind DynamicMeshTransform_0 (Dynamic%20Mesh%20Transform :Transform (struct :Rotation (0.0 -0.0 0.7071067811865477 0.7071067811865475) :Translation (-6e+01 0.0 175.0) :Scale3D (0.6 1.0 1.0))))
  (bind DynamicMeshTransform_2 (Dynamic%20Mesh%20Transform :Transform (struct :Rotation (0.0 -0.0 0.7071069999998647 0.7071065623731627) :Translation (6e+01 0.0 175.0) :Scale3D (0.6 1.0 1.0))))
  (bind DynamicMeshTransform_4 (Dynamic%20Mesh%20Transform :Transform (struct :Rotation (0.0 -0.0 0.7071069999998647 0.7071065623731627) :Translation (-6e+01 0.0 6e+01) :Scale3D (0.6 1.0 1.0))))
  (bind DynamicMeshTransform_6 (Dynamic%20Mesh%20Transform :Transform (struct :Rotation (0.0 -0.0 0.7071069999998647 0.7071065623731627) :Translation (6e+01 0.0 6e+01) :Scale3D (0.6 1.0 1.0))))
  (bind BooleanOperation_1 (Boolean%20Operation :BooleanOperation Subtract :Mode EachAWithEachBSequentially))
  (bind SpawnDynamicMesh_0 (Spawn%20Dynamic%20Mesh))
  (connect DynamicMeshTransform_1 Out BooleanOperation_1 InA)
  (connect DynamicMeshTransform_0 Out BooleanOperation_1 InB)
  (connect DynamicMeshTransform_2 Out BooleanOperation_1 InB)
  (connect DynamicMeshTransform_4 Out BooleanOperation_1 InB)
  (connect DynamicMeshTransform_6 Out BooleanOperation_1 InB)
  (connect StaticMeshToDynamicMeshElement_0 Out DynamicMeshTransform_0 In)
  (connect StaticMeshToDynamicMeshElement_0 Out DynamicMeshTransform_1 In)
  (connect StaticMeshToDynamicMeshElement_0 Out DynamicMeshTransform_2 In)
  (connect StaticMeshToDynamicMeshElement_0 Out DynamicMeshTransform_4 In)
  (connect StaticMeshToDynamicMeshElement_0 Out DynamicMeshTransform_6 In)
  (connect BooleanOperation_1 Out SpawnDynamicMesh_0 In))
