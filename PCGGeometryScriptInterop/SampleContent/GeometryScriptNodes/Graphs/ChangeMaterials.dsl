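; PCG sample graph that shows how to set and replace materials on dynamic meshes with Geometry Script interop nodes, then spawns the result.
; A default-material sphere serves as base mesh, and copies of a custom-material sphere are appended to it at jittered points on a generated sphere.
; The combined mesh's materials are read, DefaultMaterial is swapped for BasicAsset03 via a string Replace, and the list is reapplied before spawning.
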(graph ChangeMaterials
  (bind CreatePointsSphere_0 (Create%20Points%20Sphere :GeodesicSubdivisions 5 :Jitter 0.5 :Phi 24.0 :Radius 5e+02 :SphereGeneration Angle :Theta 24.0))
  (bind AppendMeshesFromPoints_0 (Append%20Meshes%20From%20Points :Mesh /Engine/BasicShapes/Sphere.Sphere :Mode DynamicMesh))
  (bind SpawnDynamicMesh_0 (Spawn%20Dynamic%20Mesh))
  ;[0.15 0.15 0.15 0.5] Simple sphere with custom material
  (bind StaticMeshToDynamicMeshElement_1 (Mesh%20To%20Dynamic%20Mesh%20Element :Mesh /Engine/BasicShapes/Sphere.Sphere))
  (bind SetMaterialsDynamicMesh_0 (Set%20Materials%20On%20Dynamic%20Mesh :Materials [/Engine/MapTemplates/Materials/BasicAsset01.BasicAsset01]))
  ;[0.15 0.15 0.15 0.5] Simple sphere with default material
  (bind StaticMeshToDynamicMeshElement_0 (Mesh%20To%20Dynamic%20Mesh%20Element :Mesh /Engine/BasicShapes/Sphere.Sphere))
  ;[0.15 0.15 0.15 0.5] Replace the default material by another one
  (bind GetMaterialsDynamicMesh_0 (Get%20Materials%20On%20Dynamic%20Mesh))
  (bind CreateConstant_0 (Create%20Constant :AttributeTypes (struct :Type SoftObjectPath :SoftObjectPathValue (struct :AssetPath (struct :PackageName /Engine/EngineMaterials/DefaultMaterial :AssetName DefaultMaterial))) :OutputTarget OldMaterial))
  (bind CreateConstant_1 (Create%20Constant :AttributeTypes (struct :Type SoftObjectPath :SoftObjectPathValue (struct :AssetPath (struct :PackageName /Engine/MapTemplates/Materials/BasicAsset03 :AssetName BasicAsset03))) :OutputTarget NewMaterial))
  (bind AttributeStringOp_0 (Attribute%20String%20Op :Operation Replace))
  ;[0.15 0.15 0.15 0.5] Apply
  (bind SetMaterialsDynamicMesh_1 (Set%20Materials%20On%20Dynamic%20Mesh :Materials [/Engine/MapTemplates/Materials/BasicAsset01.BasicAsset01]))
  (connect SetMaterialsDynamicMesh_0 Out AppendMeshesFromPoints_0 AppendDynMesh)
  (connect StaticMeshToDynamicMeshElement_0 Out AppendMeshesFromPoints_0 InDynMesh)
  (connect CreatePointsSphere_0 Out AppendMeshesFromPoints_0 InPoints)
  (connect GetMaterialsDynamicMesh_0 Out AttributeStringOp_0 In)
  (connect CreateConstant_1 Out AttributeStringOp_0 Replace)
  (connect CreateConstant_0 Out AttributeStringOp_0 Search)
  (connect AppendMeshesFromPoints_0 Out GetMaterialsDynamicMesh_0 In)
  (connect StaticMeshToDynamicMeshElement_1 Out SetMaterialsDynamicMesh_0 In)
  (connect AppendMeshesFromPoints_0 Out SetMaterialsDynamicMesh_1 In)
  (connect AttributeStringOp_0 Out SetMaterialsDynamicMesh_1 Materials)
  (connect SetMaterialsDynamicMesh_1 Out SpawnDynamicMesh_0 In))
