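; PCG sample graph that combines meshes from tagged world actors, labeled as trees, into one spawned dynamic mesh.
; Get Dynamic Mesh Data gathers all actors tagged MyTag, and two chained index filters split off the first and second, leaving any remainder as the third branch.
; Each branch gets its own rotation and downscale through Dynamic Mesh Transform.
; Merge Dynamic Meshes joins the three branches and feeds Spawn Dynamic Mesh.
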(graph CombineActors
  (bind GetDynamicMeshData_0 (Get%20Dynamic%20Mesh%20Data :ActorSelector (struct :ActorFilter AllWorldActors :ActorSelectionTag MyTag :bSelectMultiple true) :bTrackActorsOnlyWithinBounds false))
  (bind FilterDataByIndex_1 (Filter%20Data%20By%20Index :SelectedIndices "0"))
  (bind FilterDataByIndex_3 (Filter%20Data%20By%20Index :SelectedIndices "0"))
  (bind DynamicMeshTransform_7 (Dynamic%20Mesh%20Transform :Transform (struct :Rotation (0.0 -0.3826834323650896 0.0 0.9238795325112867) :Scale3D (0.4 0.4 0.4))))
  (bind DynamicMeshTransform_9 (Dynamic%20Mesh%20Transform :Transform (struct :Rotation (-0.0 0.7071067811865476 0.0 0.7071067811865477) :Scale3D (0.2 0.2 0.2))))
  (bind DynamicMeshTransform_11 (Dynamic%20Mesh%20Transform :Transform (struct :Rotation (-0.0 0.3826834323650896 0.0 0.9238795325112867) :Scale3D (0.1 0.1 0.1))))
  (bind MergeDynamicMeshes_12 (Merge%20Dynamic%20Meshes))
  (bind SpawnDynamicMesh_0 (Spawn%20Dynamic%20Mesh))
  (reroute First%20Tree FilterDataByIndex_1.InsideFilter)
  (reroute Second%20Tree FilterDataByIndex_3.InsideFilter)
  (reroute Third%20Tree FilterDataByIndex_3.OutsideFilter)
  (connect First%20Tree Out DynamicMeshTransform_7 In)
  (connect Second%20Tree Out DynamicMeshTransform_9 In)
  (connect Third%20Tree Out DynamicMeshTransform_11 In)
  (connect GetDynamicMeshData_0 Out FilterDataByIndex_1 In)
  (connect FilterDataByIndex_1 OutsideFilter FilterDataByIndex_3 In)
  (connect DynamicMeshTransform_7 Out MergeDynamicMeshes_12 In)
  (connect DynamicMeshTransform_9 Out MergeDynamicMeshes_12 In)
  (connect DynamicMeshTransform_11 Out MergeDynamicMeshes_12 In)
  (connect MergeDynamicMeshes_12 Out SpawnDynamicMesh_0 In))
