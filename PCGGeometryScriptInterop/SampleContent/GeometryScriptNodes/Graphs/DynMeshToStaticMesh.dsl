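; PCG sample graph that builds a bunny-like dynamic mesh from three spheres and scatters it as static meshes over world surfaces.
; Merged, transformed spheres are saved as a static mesh asset when SaveAsset is true.
; A Gather execution dependency holds back World Ray Hit surface sampling until that branch finishes.
; Points get a Mesh attribute assembled from AssetPath and AssetName, then random yaw and scale before attribute-driven spawning.
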(graph DynMeshToStaticMesh
  (params (SaveAsset :type Bool) (AssetName :type String :default "Bunny") (AssetPath :type String :default "/Game"))
  (bind SaveDynamicMeshToAsset_6 (Save%20Dynamic%20Mesh%20To%20Asset :CopyMeshToAssetOptions (struct :bReplaceMaterials true) :ExportParams (struct :bOpenSaveDialog false :AssetName "Bunny" :AssetPath "/Game") :AssetName AssetName :AssetPath AssetPath))
  (bind WorldRayHitQuery_7 (World%20Ray%20Hit%20Query :QueryParams (struct :CollisionChannel ECC_WorldDynamic)))
  (bind Gather_12 (Gather))
  (bind SurfaceSampler_8 (Surface%20Sampler :PointsPerSquaredMeter 0.5 :PointSteepness 0.5))
  (bind AddAttribute_9 (Add%20Attribute :OutputTarget Mesh))
  (bind TransformPoints_10 (Transform%20Points :OffsetMax (0.0 0.0 5e+01) :OffsetMin (0.0 0.0 5e+01) :RotationMax (struct :Yaw 3.6e+02) :ScaleMax (0.4 0.4 0.4) :ScaleMin (0.2 0.2 0.2)))
  (bind StaticMeshSpawner_11 (Static%20Mesh%20Spawner :MeshSelectorParameters (class:/Script/PCG.PCGMeshSelectorByAttribute :AttributeName Mesh) :MeshSelectorType /Script/PCG.PCGMeshSelectorByAttribute))
  ;[0.03457 0.15 0.07399 0.5] Spheres
  (bind StaticMeshToDynamicMeshElement_0 (Mesh%20To%20Dynamic%20Mesh%20Element :Mesh /Engine/EngineMeshes/Sphere.Sphere :OverrideMaterials [/Engine/EngineMeshes/BackgroundMaterial.BackgroundMaterial]))
  (bind StaticMeshToDynamicMeshElement_1 (Mesh%20To%20Dynamic%20Mesh%20Element :Mesh /Engine/EngineMeshes/Sphere.Sphere :OverrideMaterials [/Engine/BasicShapes/BasicShapeMaterial.BasicShapeMaterial]))
  (bind DynamicMeshTransform_1 (Dynamic%20Mesh%20Transform :Transform (struct :Rotation (-0.0 0.2588190451025208 0.0 0.9659258262890684) :Translation (1e+02 0.0 275.0) :Scale3D (0.3 0.3 1.25))))
  (bind DynamicMeshTransform_3 (Dynamic%20Mesh%20Transform :Transform (struct :Rotation (0.0 -0.1305261922200515 0.0 0.9914448613738103) :Translation (-1e+02 0.0 275.0) :Scale3D (0.3 0.3 1.25))))
  (bind MergeDynamicMeshes_4 (Merge%20Dynamic%20Meshes))
  (bind Branch_3 (Branch :bOutputToB-param SaveAsset))
  ;[0.15 0.15 0.15 0.5] Build the asset path
  (bind CreateAttribute_2 (Create%20Constant :AttributeTypes (struct :Type String :StringValue "{Path}/{Name}.{Name}")))
  (bind CreateAttribute_6 (Create%20Constant :AttributeTypes (struct :Type String :StringValue "{Path}")))
  (bind AttributeStringOp_3 (Attribute%20String%20Op :Operation Replace :Replace AssetPath))
  (bind CreateAttribute_10 (Create%20Constant :AttributeTypes (struct :Type String :StringValue "{Name}")))
  (bind AttributeStringOp_8 (Attribute%20String%20Op :Operation Replace :Replace AssetName))
  (connect AttributeStringOp_8 Out AddAttribute_9 Attributes)
  (connect SurfaceSampler_8 Out AddAttribute_9 In)
  (connect CreateAttribute_2 Out AttributeStringOp_3 In)
  (connect CreateAttribute_6 Out AttributeStringOp_3 Search)
  (connect AttributeStringOp_3 Out AttributeStringOp_8 In)
  (connect CreateAttribute_10 Out AttributeStringOp_8 Search)
  (connect MergeDynamicMeshes_4 Out Branch_3 In)
  (connect StaticMeshToDynamicMeshElement_1 Out DynamicMeshTransform_1 In)
  (connect StaticMeshToDynamicMeshElement_1 Out DynamicMeshTransform_3 In)
  (connect Branch_3 Output%20A Gather_12 Execution%20Dependency)
  (connect SaveDynamicMeshToAsset_6 OutAssetPath Gather_12 Execution%20Dependency)
  (connect WorldRayHitQuery_7 Out Gather_12 In)
  (connect DynamicMeshTransform_1 Out MergeDynamicMeshes_4 In)
  (connect DynamicMeshTransform_3 Out MergeDynamicMeshes_4 In)
  (connect StaticMeshToDynamicMeshElement_0 Out MergeDynamicMeshes_4 In)
  (connect Branch_3 Output%20B SaveDynamicMeshToAsset_6 In)
  (connect TransformPoints_10 Out StaticMeshSpawner_11 In)
  (connect Gather_12 Out SurfaceSampler_8 Surface)
  (connect AddAttribute_9 Out TransformPoints_10 In))
