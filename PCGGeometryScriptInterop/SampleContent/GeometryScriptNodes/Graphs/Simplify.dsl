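; Sample PCG graph showing two ways to write a Blueprint element that simplifies a dynamic mesh through Geometry Script.
; Two parallel branches each convert the engine cube to a dynamic mesh, reduce it to a low vertex count and spawn the result.
; One uses a streamlined element that processes the mesh directly.
; The other uses a manual element that overrides Execute and uses Copy or Steal to avoid duplicating data when it can work in place.
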(graph Simplify
  (bind StaticMeshToDynamicMeshElement_0 (Mesh%20To%20Dynamic%20Mesh%20Element :Mesh /Engine/EngineMeshes/Cube.Cube))
  (bind StaticMeshToDynamicMeshElement_1 (Mesh%20To%20Dynamic%20Mesh%20Element :Mesh /Engine/EngineMeshes/Cube.Cube))
  ;[0.15 0.15 0.15 0.5] Streamlined version that processes the Dynamic mesh
  (bind ExecuteBlueprint_1 (Execute%20Blueprint :BlueprintElementInstance (class:/PCGGeometryScriptInterop/SampleContent/GeometryScriptNodes/Blueprint/PCG_DynamicMeshSimplify.PCG_DynamicMeshSimplify_C :VertexCount 7) :BlueprintElementType /PCGGeometryScriptInterop/SampleContent/GeometryScriptNodes/Blueprint/PCG_DynamicMeshSimplify.PCG_DynamicMeshSimplify_C))
  (bind SpawnDynamicMesh_3 (Spawn%20Dynamic%20Mesh))
  ;[0.15 0.15 0.15 0.5] More customizable version that overrides Execute like a
  ; normal PCG BP element, but uses Copy or Steal to
  ; avoid duplicating the data if we can use it in place.
  (bind ExecuteBlueprint_3 (Execute%20Blueprint :BlueprintElementInstance (class:/PCGGeometryScriptInterop/SampleContent/GeometryScriptNodes/Blueprint/PCG_DynamicMeshSimplify_ManualProcessing.PCG_DynamicMeshSimplify_ManualProcessing_C :VertexCount 7) :BlueprintElementType /PCGGeometryScriptInterop/SampleContent/GeometryScriptNodes/Blueprint/PCG_DynamicMeshSimplify_ManualProcessing.PCG_DynamicMeshSimplify_ManualProcessing_C))
  (bind SpawnDynamicMesh_5 (Spawn%20Dynamic%20Mesh))
  (connect StaticMeshToDynamicMeshElement_0 Out ExecuteBlueprint_1 In)
  (connect StaticMeshToDynamicMeshElement_1 Out ExecuteBlueprint_3 In)
  (connect ExecuteBlueprint_1 Out SpawnDynamicMesh_3 In)
  (connect ExecuteBlueprint_3 Out SpawnDynamicMesh_5 In))
