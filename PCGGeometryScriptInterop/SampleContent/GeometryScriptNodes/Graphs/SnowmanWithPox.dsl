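; PCG Geometry Script interop sample that builds a snowman dynamic mesh covered in small sphere "pox" and spawns it.
; Engine primitives are transformed and merged into head, arms, hat and body.
; The merged mesh is Poisson-sampled into points, which are scaled down and filtered by normal to drop those facing nearly straight up or down.
; PoxFromDynMesh selects whether a dynamic mesh or an attribute-referenced static mesh is appended at those points.
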(graph SnowmanWithPox
  (params (PoxFromDynMesh :type Bool :default true))
  (bind SpawnDynamicMesh_27 (Spawn%20Dynamic%20Mesh))
  ;[0.9134 0.9635 0.9635 0.5] WhiteSphere
  (bind StaticMeshToDynamicMeshElement_0 (Mesh%20To%20Dynamic%20Mesh%20Element :Mesh /Engine/BasicShapes/Sphere.Sphere :OverrideMaterials [/Engine/BasicShapes/BasicShapeMaterial.BasicShapeMaterial]))
  (bind DynamicMeshTransform_2 (Dynamic%20Mesh%20Transform :Transform (struct :Scale3D (3.0 3.0 3.0))))
  ;[0.02669 0.132 0.15 0.5] Head
  (bind StaticMeshToDynamicMeshElement_2 (Mesh%20To%20Dynamic%20Mesh%20Element :Mesh /Engine/BasicShapes/Sphere.Sphere :OverrideMaterials [/Engine/EngineMeshes/BackgroundMaterial.BackgroundMaterial]))
  (bind MergeDynamicMeshes_5 (Merge%20Dynamic%20Meshes))
  ;;[0.15 0.1275 0.01356 0.5] Nose
  (bind StaticMeshToDynamicMeshElement_4 (Mesh%20To%20Dynamic%20Mesh%20Element :Mesh /Engine/BasicShapes/Cone.Cone :OverrideMaterials [/Engine/MapTemplates/Materials/BasicAsset01.BasicAsset01]))
  (bind DynamicMeshTransform_4 (Dynamic%20Mesh%20Transform :Transform (struct :Rotation (-0.707107 -0.0 0.0 0.707107) :Translation (0.0 2e+02 0.0) :Scale3D (0.5 0.5 1.0))))
  ;;[0.15 0.03131 0.0725 0.5] Eyes
  (bind DynamicMeshTransform_6 (Dynamic%20Mesh%20Transform :Transform (struct :Translation (5e+01 1.2e+02 8e+01) :Scale3D (0.3 0.3 0.3))))
  (bind DynamicMeshTransform_7 (Dynamic%20Mesh%20Transform :Transform (struct :Translation (-5e+01 1.2e+02 8e+01) :Scale3D (0.3 0.3 0.3))))
  ;;[0.15 0.05255 0.07209 0.5] Hat
  (bind StaticMeshToDynamicMeshElement_6 (Mesh%20To%20Dynamic%20Mesh%20Element :Mesh /Engine/BasicShapes/Cylinder.Cylinder :OverrideMaterials [/Engine/EngineMeshes/BackgroundMaterial.BackgroundMaterial]))
  (bind DynamicMeshTransform_17 (Dynamic%20Mesh%20Transform :Transform (struct :Translation (0.0 0.0 -5e+01) :Scale3D (1.5 1.5 0.1))))
  (bind MergeDynamicMeshes_19 (Merge%20Dynamic%20Meshes))
  (bind DynamicMeshTransform_18 (Dynamic%20Mesh%20Transform :Transform (struct :Translation (0.0 0.0 2e+02))))
  ;[0.09172 0.15 0.01852 0.5] Full Body
  (bind DynamicMeshTransform_1 (Dynamic%20Mesh%20Transform :Transform (struct :Translation (0.0 0.0 3.5e+02) :Scale3D (0.5 0.5 0.5))))
  (bind MergeDynamicMeshes_2 (Merge%20Dynamic%20Meshes))
  ;[0.15 0.08648 0.02403 0.5] Middle Body
  (bind MergeDynamicMeshes_38 (Merge%20Dynamic%20Meshes))
  (bind DynamicMeshTransform_8 (Dynamic%20Mesh%20Transform :Transform (struct :Translation (0.0 0.0 2e+02) :Scale3D (0.7 0.7 0.7))))
  ;;[0.03008 0.15 0.01784 0.5] Arms
  (bind StaticMeshToDynamicMeshElement_34 (Mesh%20To%20Dynamic%20Mesh%20Element :Mesh /Engine/EngineMeshes/Cube.Cube :OverrideMaterials [/Engine/EngineMeshes/BackgroundMaterial.BackgroundMaterial]))
  (bind DynamicMeshTransform_35 (Dynamic%20Mesh%20Transform :Transform (struct :Rotation (0.16042999720435944 -0.37686961114246087 0.06645228065352285 0.9098437264664108) :Translation (175.0 5e+01 0.0) :Scale3D (0.05 0.05 1.0))))
  (bind DynamicMeshTransform_37 (Dynamic%20Mesh%20Transform :Transform (struct :Rotation (0.16042999720435958 0.3768696111424612 -0.066452280653523 0.9098437264664101) :Translation (-175.0 5e+01 0.0) :Scale3D (0.05 0.05 1.0))))
  ;[0.02337 0.1168 0.15 0.5] Sampling points on the dynamic mesh and filtering the normals
  ; Show Interop with points
  (bind MeshSampler_10 (Mesh%20Sampler :bExtractMeshFromInput true :bHasDefaultColorOverride false :PointSteepness 0.5 :SamplingMethod PoissonSampling :SamplingOptions (struct :SamplingRadius 2e+01)))
  (bind CreateAttribute_12 (Create%20Constant :AttributeTypes (struct :DoubleValue 0.02)))
  (bind AttributeMathsOp_11 (Attribute%20Maths%20Op :InputSource1 $Scale :Operation Multiply))
  (bind CreateAttribute_21 (Create%20Constant :AttributeTypes (struct :Type Vector :VectorValue (0.0 0.0 1.0))))
  (bind AttributeVectorOp_20 (Attribute%20Vector%20Op :InputSource1 $Rotation.Up :Operation Dot :OutputTarget $Density))
  (bind AttributeFilterRange_13 (Filter%20Attribute%20Elements%20by%20Range :MaxThreshold (struct :bUseConstantThreshold true :ThresholdAttribute $Density :bUseSpatialQuery false :AttributeTypes (struct :DoubleValue 0.8)) :MinThreshold (struct :bUseConstantThreshold true :ThresholdAttribute $Density :bUseSpatialQuery false :AttributeTypes (struct :DoubleValue -0.8))))
  (bind AddAttribute_0 (Add%20Attribute :AttributeTypes (struct :Type SoftObjectPath :SoftObjectPathValue (struct :AssetPath (struct :PackageName /Engine/EngineMeshes/Sphere :AssetName Sphere))) :OutputTarget Mesh))
  ;[0.15 0.02969 0.02969 0.5] Append the mesh (either static from attrbiute, or coming from another dyn mesh) at the points transform.
  (bind Branch_3 (Branch :bOutputToB-param PoxFromDynMesh))
  (bind AppendMeshesFromPoints_14 (Append%20Meshes%20From%20Points :Mesh /Engine/EngineMeshes/Sphere.Sphere :MeshAttribute Mesh :Mode StaticMeshFromAttribute))
  (bind StaticMeshToDynamicMeshElement_7 (Mesh%20To%20Dynamic%20Mesh%20Element :Mesh /Engine/BasicShapes/Sphere.Sphere :OverrideMaterials [/Engine/MapTemplates/Materials/BasicAsset03.BasicAsset03]))
  (bind DynamicMeshTransform_11 (Dynamic%20Mesh%20Transform :Transform (struct :Scale3D (3.0 3.0 3.0))))
  (bind AppendMeshesFromPoints_6 (Append%20Meshes%20From%20Points :Mesh /Engine/EngineMeshes/Sphere.Sphere :MeshAttribute Mesh :Mode DynamicMesh))
  (connect AttributeFilterRange_13 InsideFilter AddAttribute_0 In)
  (connect DynamicMeshTransform_11 Out AppendMeshesFromPoints_6 AppendDynMesh)
  (connect Branch_3 Output%20B AppendMeshesFromPoints_6 InDynMesh)
  (connect AddAttribute_0 Out AppendMeshesFromPoints_6 InPoints)
  (connect Branch_3 Output%20A AppendMeshesFromPoints_14 InDynMesh)
  (connect AddAttribute_0 Out AppendMeshesFromPoints_14 InPoints)
  (connect AttributeVectorOp_20 Out AttributeFilterRange_13 In)
  (connect MeshSampler_10 Out AttributeMathsOp_11 InA)
  (connect CreateAttribute_12 Out AttributeMathsOp_11 InB)
  (connect AttributeMathsOp_11 Out AttributeVectorOp_20 InA)
  (connect CreateAttribute_21 Out AttributeVectorOp_20 InB)
  (connect MergeDynamicMeshes_2 Out Branch_3 In)
  (connect MergeDynamicMeshes_5 Out DynamicMeshTransform_1 In)
  (connect StaticMeshToDynamicMeshElement_0 Out DynamicMeshTransform_2 In)
  (connect StaticMeshToDynamicMeshElement_4 Out DynamicMeshTransform_4 In)
  (connect StaticMeshToDynamicMeshElement_2 Out DynamicMeshTransform_6 In)
  (connect StaticMeshToDynamicMeshElement_2 Out DynamicMeshTransform_7 In)
  (connect MergeDynamicMeshes_38 Out DynamicMeshTransform_8 In)
  (connect StaticMeshToDynamicMeshElement_7 Out DynamicMeshTransform_11 In)
  (connect StaticMeshToDynamicMeshElement_6 Out DynamicMeshTransform_17 In)
  (connect MergeDynamicMeshes_19 Out DynamicMeshTransform_18 In)
  (connect StaticMeshToDynamicMeshElement_34 Out DynamicMeshTransform_35 In)
  (connect StaticMeshToDynamicMeshElement_34 Out DynamicMeshTransform_37 In)
  (connect DynamicMeshTransform_1 Out MergeDynamicMeshes_2 In)
  (connect DynamicMeshTransform_2 Out MergeDynamicMeshes_2 In)
  (connect DynamicMeshTransform_8 Out MergeDynamicMeshes_2 In)
  (connect DynamicMeshTransform_2 Out MergeDynamicMeshes_5 In)
  (connect DynamicMeshTransform_4 Out MergeDynamicMeshes_5 In)
  (connect DynamicMeshTransform_6 Out MergeDynamicMeshes_5 In)
  (connect DynamicMeshTransform_7 Out MergeDynamicMeshes_5 In)
  (connect DynamicMeshTransform_18 Out MergeDynamicMeshes_5 In)
  (connect DynamicMeshTransform_17 Out MergeDynamicMeshes_19 In)
  (connect StaticMeshToDynamicMeshElement_6 Out MergeDynamicMeshes_19 In)
  (connect DynamicMeshTransform_2 Out MergeDynamicMeshes_38 In)
  (connect DynamicMeshTransform_35 Out MergeDynamicMeshes_38 In)
  (connect DynamicMeshTransform_37 Out MergeDynamicMeshes_38 In)
  (connect MergeDynamicMeshes_2 Out MeshSampler_10 In)
  (connect AppendMeshesFromPoints_6 Out SpawnDynamicMesh_27 In)
  (connect AppendMeshesFromPoints_14 Out SpawnDynamicMesh_27 In))
